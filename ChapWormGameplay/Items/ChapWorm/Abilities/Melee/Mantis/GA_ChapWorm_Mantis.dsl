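; Mantis melee gameplay ability for the ChapWorm item that adjusts the player camera and adapts melee techniques to Overpower (kinetic) mode.
; Activation pushes a tagged camera mode override onto the player camera manager, and ability end removes it.
; The Execution notify window applies the melee activation GameplayEffect to the owner.
; The ChapWorm.Overpower tag switches the dynamic montage name from Base to Surge and raises the effect container level.
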(event Ability|EventActivateAbility
  ;[0.00516 0.1302 0.04559] ActivateAbility: Add camera mode.
  (|CallParentFunction)
  (Default|AddCameraMode))

(event Ability|EventOnEndAbility
  (params (bWasCancelled :type Boolean))
  ;[0.1094 0.06834 0.01647] EndAbility: Remove camera mode.
  (Utilities|FlowControl|Sequence
    (:then_0
      (Default|RemoveCameraMode))
    (:then_1
      (|CallParentFunction bWasCancelled))))

(event Mantis|EventOnTechniqueNotifyWindowStarted
  (params (NotifyWindow :type EFortMantisNotifyWindow%20Enum :default "Invalid") (Notify :type Mantis%20Window%20Object%20Reference))
  ;[0.0174 0.1198 0.1162] OnTechniqueNotifyWindowStarted: When a technique execution starts, apply mantis activation GE to the player.
  (|CallParentFunction NotifyWindow Notify)
  (if (Utilities|Enum|Equal(Enum) NotifyWindow "Execution")
    (Ability|ApplyGameplayEffectToOwner "/ChapWormGameplay/Items/ChapWorm/Abilities/Melee/GameplayEffects/GE_ChapWorm_Melee_Activate.GE_ChapWorm_Melee_Activate_C")))

(fn AddCameraMode
  ;[0.000237 0.04228 0.05729] AddCameraMode: Add the desired camera mode to the player pawn.
  (bind _outfortplayerpawn (CallFunction|GetAvatarasValidFortPlayerPawn)
    (:ValidFortPlayerPawn
      (Utilities|IsValid (Fortnite|GetFortPlayerControllerfromActor _outfortplayerpawn)
        (:Is%20Valid
          (bind _playercameramanager (Variables|/Script/Engine.PlayerController|GetPlayerCameraManager (Fortnite|GetFortPlayerControllerfromActor _outfortplayerpawn))
            (:then
              (bind _asfort_player_camera_base (Utilities|Casting|CastToFortPlayerCameraBase _playercameramanager))
              (Camera|AddCameraModeOverride _asfort_player_camera_base (Variables|Camera|GetCameraMode) (Variables|Camera|GetCameraModeTag) 2 0 0))
            (:else)))))))

(fn GetDynamicMontageNameForTechnique
  (params (TechniqueName :type Name :default "None"))
  (bind _returnvalue (CallFunction|GetAvatarasFortPlayerPawn))
  ;[0.0277 0.1106 0.1823] GetDynamicMontageNameForTechnique: Return Surge if the player is in kinetic mode.
  (Utilities|FlowControl|Sequence
    (:then_0
      (Utilities|IsValid _returnvalue
        (:Is%20Valid
          (if (GameplayTags|HasMatchingGameplayTag _returnvalue "(TagName=\"ChapWorm.Overpower\")")
            (return "Surge")
            (else
              (return "Base"))))))
    (:then_1
      (return "None"))))

(fn GetLevelForGameplayEffectContainer
  (params (ApplicationTag :type Gameplay%20Tag%20Structure))
  (bind _returnvalue (CallFunction|GetAvatarasFortPlayerPawn))
  (Utilities|FlowControl|Sequence
    (:then_0
      (Utilities|IsValid _returnvalue
        (:Is%20Valid
          (return (select (GameplayTags|HasMatchingGameplayTag _returnvalue "(TagName=\"ChapWorm.Overpower\")") 1 0)))))
    (:then_1
      (return 0))))

(fn OverrideFailedReason
  (params (FailedReason :type Gameplay%20Tag%20Container%20Structure%20%28by%20ref%29))
  (return FailedReason))

(fn RemoveCameraMode
  ;[0.009367 0.09587 0.1042] RemoveCameraMode: Remove the applied camera mode.
  (bind _outfortplayerpawn (CallFunction|GetAvatarasValidFortPlayerPawn)
    (:ValidFortPlayerPawn
      (Utilities|IsValid (Fortnite|GetFortPlayerControllerfromActor _outfortplayerpawn)
        (:Is%20Valid
          (bind _playercameramanager (Variables|/Script/Engine.PlayerController|GetPlayerCameraManager (Fortnite|GetFortPlayerControllerfromActor _outfortplayerpawn))
            (:then
              (bind _asfort_player_camera_base (Utilities|Casting|CastToFortPlayerCameraBase _playercameramanager))
              (Camera|RemoveCameraModeOverride _asfort_player_camera_base (Variables|Camera|GetCameraModeTag) 0 true))
            (:else)))))))
